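# Translucent UI master material that draws an angled backing panel with an accent bar along its slanted edge.
# Rotated UV ramps form two edge masks: one trims the panel, the other, offset by bar width plus edge offset, separates bar from panel.
# A resolution sharpness function scales edge hardness by pixel size.
# Static switches enable color gradients on bar and panel and a panel alpha fade, with selectable axes and overall alpha.
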Material /FortniteCoreUI/Frontend/Spatial/Art/M_UI_BPSpatial_AngledBacking_Master
graph {
 properties {
  bUsedWithStaticMesh=false
  blendMode=BLEND_Translucent
  materialDomain=MD_UI
  pixelDepthOffsetMode=PDOM_Legacy
  usedShadingModels=Default Lit
 }
 nodes [
  n1 = Add(A=TextureCoordinate(), B=ScalarParameter() {group=Panel, parameterName=AlphaFade_Offset})
  n2 = ScalarParameter() {defaultValue=1, group=Panel, parameterName=AlphaFade_Ramp}
  n3 = LinearInterpolate(
   A=ScalarParameter() {defaultValue=0.5, group=Panel, parameterName=Fade_Start_Alpha}
   B=ScalarParameter() {defaultValue=1, group=Panel, parameterName=Fade_Stop_Alpha}
   Alpha=StaticComponentMaskParameter(None=Power(Base=@n1, Exp=@n2)) {defaultG=true, group=Panel, parameterName=PanelAlphaFade_Direction}
  )
  n4 = ScalarParameter() {defaultValue=0.75, group=Panel, parameterName=Panel_Alpha}
  n5 = Multiply(A=ScalarParameter() {defaultValue=3, group=Bar, parameterName=Bar_Rotation}) {constB=-0.01}
  n6 = TextureCoordinate()
  n7 = Add(
   A=ScalarParameter() {defaultValue=20, group=Bar, parameterName=Bar_Width}
   B=ScalarParameter() {defaultValue=10, group=Panel, parameterName=Panel_EdgeOffset}
  )
  n8 = ComponentMask(None=Add(
    A=MaterialFunctionCall!(Rotation_Fraction=@n5, TexCoord=@n6) {materialFunction=/FortUILibrary/Foundation/Art_ToolKit/AngledBoxes/MF_UI_Rotator_Fractions}.Result
    B=Multiply(A=@n7) {constB=-0.01}
   )) {r=true}
  n9 = MaterialFunctionCall!(Native_U_PixelSize=ScalarParameter() {defaultValue=256, parameterName=U_PixelSize}) {materialFunction=/FortUILibrary/Foundation/Art_ToolKit/Tools/M_UI_Resolution_Sharpness_Correction}
  n10 = Multiply(A=@n9.Result, B=ScalarParameter() {defaultValue=64, group=Bar, parameterName=Bar_EdgeHardness})
  n11 = OneMinus(None=Saturate(None=Multiply(A=@n8, B=@n10)))
  n12 = Saturate(None=LinearInterpolate(
    A=StaticSwitchParameter(True=@n3, False=@n4) {group=Panel, parameterName=UsePanel_Fade_Alpha}
    B=ScalarParameter() {defaultValue=1, group=Bar, parameterName=Bar_Alpha}
    Alpha=@n11
   ))
  n13 = ScalarParameter() {defaultValue=2.5, group=Panel, parameterName=Panel_EdgeRotation}
  n14 = Add(
   A=MaterialFunctionCall(Rotation_Fraction=Multiply(A=@n13) {constB=-0.01}, TexCoord=TextureCoordinate()) {materialFunction=&MF_UI_Rotator_Fractions}.Result
   B=Multiply(A=ScalarParameter() {defaultValue=10, group=Panel, parameterName=Panel_EdgeOffset}) {constB=-0.01}
  )
  n15 = ScalarParameter() {defaultValue=64, group=Panel, parameterName=Panel_EdgeHardness}
  n16 = Saturate(None=Multiply(A=ComponentMask(None=@n14) {r=true}, B=Multiply(A=@n9.Result, B=@n15)))
  Saturate(None=Multiply(A=LinearInterpolate(B=@n12, Alpha=@n16), B=ScalarParameter() {defaultValue=1, parameterName=Overall_Alpha}))
  n17 = VectorParameter() {parameterName=BarColor_1, defaultValue=<rgba 1 0 0 1>}
  n18 = TextureCoordinate()
  n19 = Add(A=@n18, B=ScalarParameter() {group=Bar, parameterName=Bar_Gradient_Offset})
  n20 = ScalarParameter() {defaultValue=1, group=Bar, parameterName=Bar_Gradient_Ramp}
  n21 = LinearInterpolate(
   A=VectorParameter() {parameterName=BarColor_2, defaultValue=<rgba 0 1 0 1>}.RGB
   B=@n17.RGB
   Alpha=StaticComponentMaskParameter(None=Power(Base=@n19, Exp=@n20)) {defaultG=true, group=Bar, parameterName=Bar_GadientDirection}
  )
  n22 = VectorParameter() {group=Panel, parameterName=PanelColor_1, defaultValue=<rgba 0 0 1 1>}
  n23 = Add(A=@n18, B=ScalarParameter() {group=Panel, parameterName=Panel_Gradient_Offset})
  n24 = ScalarParameter() {defaultValue=1, group=Panel, parameterName=Panel_Gradient_Ramp}
  n25 = LinearInterpolate(
   A=VectorParameter() {group=Panel, parameterName=PanelColor_2, defaultValue=<rgba 1 1 0 1>}.RGB
   B=@n22.RGB
   Alpha=StaticComponentMaskParameter(None=Power(Base=@n23, Exp=@n24)) {defaultG=true, group=Panel, parameterName=Panel_GadientDirection}
  )
  Saturate(None=LinearInterpolate(
    A=StaticSwitchParameter(True=@n25, False=@n22.RGB) {defaultValue=true, group=Panel, parameterName=Use_PanelGradient}
    B=StaticSwitchParameter(True=@n21, False=@n17.RGB) {defaultValue=true, group=Bar, parameterName=Use_BarGradient}
    Alpha=@n11
   ))
 ]
}
