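# Translucent UI material, named for factions, that draws a panel split horizontally into top and bottom color halves.
# A steep UV ramp makes the hard split at mid-height.
# HighlightTop drives the halves inversely between base and highlight colors, with Use Highlight blending from full highlight to that result.
# A second mask trims left and right edges by offsets, with shared Sharpness feathering and a Background_Opacity multiplier.
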Material /FortniteCoreUI/Frontend/Spatial/Art/M_Ui_Faction
graph {
 properties {
  bUsedWithStaticMesh=false
  blendMode=BLEND_Translucent
  materialDomain=MD_UI
  pixelDepthOffsetMode=PDOM_Legacy
  usedShadingModels=Default Lit
 }
 nodes [
  n1 = VectorParameter() {parameterName=RGB_Top_Highlight, defaultValue=<rgba 1 1 1 1>}
  n2 = VectorParameter() {parameterName=RGB_Top_Base, defaultValue=<rgba 0.6979 0.6979 0.6979 1>}
  n3 = ScalarParameter() {parameterName=HighlightTop, sliderMax=1}
  n4 = VectorParameter() {parameterName=RGB_Bottom_Highlight, defaultValue=<rgba 0.01 0.01 0.01 0>}
  n5 = VectorParameter() {parameterName=RGB_Bottom_Base, defaultValue=<rgba 0.0573 0.0573 0.0573 0>}
  n6 = OneMinus(None=ScalarParameter() {parameterName=HighlightTop})
  n7 = TextureCoordinate()
  n8 = ComponentMask(None=Add(A=@n7, B=Constant() {r=-0.5})) {g=true}
  n9 = Constant() {r=1024}
  LinearInterpolate(
   A=LinearInterpolate(A=@n1.RGB, B=LinearInterpolate(A=@n2.RGB, B=@n1.RGB, Alpha=@n3), Alpha=ScalarParameter() {parameterName=Use Highlight, sliderMax=1})
   B=LinearInterpolate(A=@n4.RGB, B=LinearInterpolate(A=@n5.RGB, B=@n4.RGB, Alpha=@n6), Alpha=ScalarParameter() {parameterName=Use Highlight})
   Alpha=Saturate(None=Multiply(A=@n8, B=@n9))
  )
  n10 = Add(A=@n7, B=ScalarParameter() {defaultValue=-0.05, group=Background, parameterName=Background_LeftOffset})
  n11 = ScalarParameter() {defaultValue=128, group=Feather, parameterName=Sharpness}
  n12 = Saturate(None=Multiply(A=ComponentMask(None=@n10) {r=true}, B=@n11))
  n13 = ComponentMask(None=Add(A=@n7, B=ScalarParameter() {defaultValue=0.05, group=Background, parameterName=Background_RightOffset})) {r=true}
  n14 = Saturate(None=Multiply(A=OneMinus(None=@n13), B=@n11))
  Multiply(A=Saturate(None=Multiply(A=@n12, B=@n14)), B=ScalarParameter() {defaultValue=1, group=Background, parameterName=Background_Opacity})
 ]
}
