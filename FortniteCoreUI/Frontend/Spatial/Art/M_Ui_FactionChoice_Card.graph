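# Translucent UI material that procedurally draws a faction choice card with a main panel and top and bottom strips.
# Rotated, offset UVs scaled by a Sharpness parameter form angled edge masks, avoiding textures.
# Opacity combines strip and panel masks, clipped at the sides, with ToBottom_AnimOffset insetting the strips horizontally.
# Color fills the panel and its inset area with two tints, and elsewhere a sine-based horizontal three-color gradient.
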Material /FortniteCoreUI/Frontend/Spatial/Art/M_Ui_FactionChoice_Card
graph {
 properties {
  bUsedWithStaticMesh=false
  blendMode=BLEND_Translucent
  materialDomain=MD_UI
  pixelDepthOffsetMode=PDOM_Legacy
  usedShadingModels=Default Lit
 }
 nodes [
  n1 = TextureCoordinate()
  n2 = Rotator(Coordinate=@n1, Time=Divide(A=ScalarParameter() {defaultValue=-3.25, group=Top, parameterName=Top_TopAngle}) {constB=360}) {centerX=1, speed=6.2832}
  n3 = ScalarParameter() {defaultValue=-0.06, group=Top, parameterName=Top_TopOffset}
  n4 = ScalarParameter() {defaultValue=512, group=Feather, parameterName=Sharpness}
  n5 = Multiply(A=ComponentMask(None=Add(A=@n2, B=@n3)) {g=true}, B=@n4)
  n6 = ScalarParameter() {defaultValue=-2.6, group=Top, parameterName=Top_BottomAngle}
  n7 = Add(
   A=Rotator(Coordinate=@n1, Time=Divide(A=@n6) {constB=360}) {centerX=0, speed=6.2832}
   B=ScalarParameter() {defaultValue=0.95, group=Top, parameterName=Top_BottomOffset}
  )
  n8 = Multiply(A=OneMinus(None=ComponentMask(None=@n7) {g=true}), B=@n4)
  n9 = Saturate(None=Multiply(A=Saturate(None=@n5), B=Saturate(None=@n8)))
  n10 = Rotator(Coordinate=@n1, Time=Divide(A=ScalarParameter() {defaultValue=4, group=Bottom, parameterName=Bottom_TopAngle}) {constB=360}) {centerX=1, speed=6.2832}
  n11 = ScalarParameter() {defaultValue=-0.76, group=Bottom, parameterName=Bottom_TopOffset}
  n12 = ScalarParameter() {defaultValue=512, group=Feather, parameterName=Sharpness}
  n13 = Multiply(A=ComponentMask(None=Add(A=@n10, B=@n11)) {g=true}, B=@n12)
  n14 = ScalarParameter() {defaultValue=6, group=Bottom, parameterName=Bottom_BottomAngle}
  n15 = Add(
   A=Rotator(Coordinate=@n1, Time=Divide(A=@n14) {constB=360}) {centerX=0, speed=6.2832}
   B=ScalarParameter() {defaultValue=0.0005, group=Bottom, parameterName=Bottom_BottomOffset}
  )
  n16 = Multiply(A=OneMinus(None=ComponentMask(None=@n15) {g=true}), B=@n12)
  n17 = Saturate(None=Multiply(A=Saturate(None=@n13), B=Saturate(None=@n16)))
  n18 = ScalarParameter() {defaultValue=1, group=Top, parameterName=Top_Opacity, sliderMax=1}
  n19 = Saturate(None=Subtract(A=Add(A=@n9, B=@n17), B=OneMinus(None=@n18)))
  n20 = ScalarParameter() {defaultValue=0.25, group=Top, parameterName=ToBottom_AnimOffset, sliderMax=0.5}
  n21 = Multiply(A=@n20) {constB=-1}
  n22 = ScalarParameter() {defaultValue=512, group=Feather, parameterName=Sharpness}
  n23 = Multiply(A=ComponentMask(None=Add(A=@n1, B=@n21)) {r=true}, B=@n22)
  n24 = Add(A=@n1, B=@n20)
  n25 = Multiply(A=OneMinus(None=ComponentMask(None=@n24) {r=true}), B=@n22)
  n26 = Saturate(None=Multiply(A=Saturate(None=@n23), B=Saturate(None=@n25)))
  n27 = ScalarParameter() {defaultValue=-2.5, group=MainPannel, parameterName=Main_TopAngle}
  n28 = Add(
   A=Rotator(Coordinate=@n1, Time=Divide(A=@n27) {constB=360}) {centerX=1, speed=6.2832}
   B=ScalarParameter() {defaultValue=-0.12, group=MainPannel, parameterName=Main_TopOffset}
  )
  n29 = ScalarParameter() {defaultValue=512, group=Feather, parameterName=Sharpness}
  n30 = Saturate(None=Multiply(A=ComponentMask(None=@n28) {g=true}, B=@n29))
  n31 = Divide(A=ScalarParameter() {defaultValue=4, group=MainPannel, parameterName=Main_BottomAngle}) {constB=360}
  n32 = ComponentMask(None=Add(
    A=Rotator(Coordinate=@n1, Time=@n31) {centerX=0, speed=6.2832}
    B=ScalarParameter() {defaultValue=0.2, group=MainPannel, parameterName=Main_BottomOffset}
   )) {g=true}
  n33 = Saturate(None=Multiply(A=OneMinus(None=@n32), B=@n29))
  n34 = Subtract(
   A=Saturate(None=Multiply(A=@n30, B=@n33))
   B=OneMinus(None=ScalarParameter() {defaultValue=0.45, group=MainPannel, parameterName=Main_Opacity, sliderMax=1})
  )
  n35 = Add(A=@n1, B=ScalarParameter() {defaultValue=-0.005, group=LeftAndRight, parameterName=LeftOffset})
  n36 = ScalarParameter() {defaultValue=512, group=Feather, parameterName=Sharpness}
  n37 = Saturate(None=Multiply(A=ComponentMask(None=@n35) {r=true}, B=@n36))
  n38 = ComponentMask(None=Add(A=@n1, B=ScalarParameter() {defaultValue=0.005, group=LeftAndRight, parameterName=RightOffset})) {r=true}
  n39 = Saturate(None=Multiply(A=OneMinus(None=@n38), B=@n36))
  DotProduct(A=Add(A=DotProduct(A=@n19, B=@n26), B=Saturate(None=@n34)), B=Saturate(None=Multiply(A=@n37, B=@n39)))
  n40 = Rotator(Coordinate=@n1, Time=Divide(A=ScalarParameter() {defaultValue=-2.5, group=MainPannel, parameterName=Main_Inner_TopAngle}) {constB=360}) {centerX=1, speed=6.2832}
  n41 = ScalarParameter() {defaultValue=-0.4, group=MainPannel, parameterName=Main_Inner_TopOffset}
  n42 = ScalarParameter() {defaultValue=512, group=Feather, parameterName=Sharpness}
  n43 = Multiply(A=ComponentMask(None=Add(A=@n40, B=@n41)) {g=true}, B=@n42)
  n44 = ScalarParameter() {defaultValue=4, group=MainPannel, parameterName=Main_Inner_BottomAngle}
  n45 = Add(
   A=Rotator(Coordinate=@n1, Time=Divide(A=@n44) {constB=360}) {centerX=0, speed=6.2832}
   B=ScalarParameter() {defaultValue=0.25, group=MainPannel, parameterName=Main_Inner_BottomOffset}
  )
  n46 = Multiply(A=OneMinus(None=ComponentMask(None=@n45) {g=true}), B=@n42)
  n47 = Saturate(None=Multiply(A=Saturate(None=@n43), B=Saturate(None=@n46)))
  n48 = ScalarParameter() {defaultValue=-0.05, group=LeftAndRight, parameterName=Inner_LeftOffset}
  n49 = ScalarParameter() {defaultValue=512, group=Feather, parameterName=Sharpness}
  n50 = Multiply(A=ComponentMask(None=Add(A=@n1, B=@n48)) {r=true}, B=@n49)
  n51 = Add(A=@n1, B=ScalarParameter() {defaultValue=0.05, group=LeftAndRight, parameterName=Inner_RightOffset})
  n52 = Multiply(A=OneMinus(None=ComponentMask(None=@n51) {r=true}), B=@n49)
  n53 = Saturate(None=Multiply(A=Saturate(None=@n50), B=Saturate(None=@n52)))
  n54 = VectorParameter() {group=Gradient, parameterName=RGB03, defaultValue=<rgba 0.9911 0.8879 0.521 1>}
  n55 = VectorParameter() {group=Gradient, parameterName=RGB04, defaultValue=<rgba 0.8632 0.5776 0.0865 1>}
  n56 = ScalarParameter() {group=Gradient, parameterName=Gradientoffset}
  n57 = Sine(None=ComponentMask(None=Add(A=@n1, B=@n56)) {r=true}) {period=2}
  n58 = Saturate(None=Add(A=@n57, B=ScalarParameter() {group=Gradient, parameterName=OuterGradientSIze}))
  n59 = Multiply(
   A=Add(A=@n57, B=ScalarParameter() {defaultValue=-0.6228, group=Gradient, parameterName=InnerGradientSize}) {constB=-0.75}
   B=ScalarParameter() {defaultValue=10, group=Gradient, parameterName=InnerGradientStrenght}
  )
  n60 = Rotator(Coordinate=@n1, Time=Divide(A=ScalarParameter() {defaultValue=-2.5, group=MainPannel, parameterName=Main_TopAngle}) {constB=360}) {centerX=1, speed=6.2832}
  n61 = ScalarParameter() {defaultValue=-0.12, group=MainPannel, parameterName=Main_TopOffset}
  n62 = ComponentMask(None=Add(A=Add(A=@n60, B=@n61)) {constB=0.005}) {g=true}
  n63 = ScalarParameter() {defaultValue=512, group=Feather, parameterName=Sharpness}
  n64 = ScalarParameter() {defaultValue=4, group=MainPannel, parameterName=Main_BottomAngle}
  n65 = Add(
   A=Rotator(Coordinate=@n1, Time=Divide(A=@n64) {constB=360}) {centerX=0, speed=6.2832}
   B=ScalarParameter() {defaultValue=0.2, group=MainPannel, parameterName=Main_BottomOffset}
  )
  n66 = OneMinus(None=ComponentMask(None=Add(A=@n65) {constB=-0.005}) {g=true})
  n67 = Multiply(A=Saturate(None=Multiply(A=@n62, B=@n63)), B=Saturate(None=Multiply(A=@n66, B=@n63)))
  LinearInterpolate(
   A=LinearInterpolate(
    A=VectorParameter() {group=MainPannel, parameterName=RGB01, defaultValue=<rgba 0 0.0578 0.5972 1>}.RGB
    B=VectorParameter() {group=MainPannel, parameterName=RGB02, defaultValue=<rgba 0 0.0908 1 1>}.RGB
    Alpha=DotProduct(A=@n47, B=@n53)
   )
   B=LinearInterpolate(
    A=LinearInterpolate(A=@n54.RGB, B=@n55.RGB, Alpha=@n58)
    B=VectorParameter() {group=Gradient, parameterName=RGB05, defaultValue=<rgba 0.3005 0.147 0.0075 1>}.RGB
    Alpha=Saturate(None=@n59)
   )
   Alpha=OneMinus(None=Saturate(None=@n67))
  )
 ]
}
