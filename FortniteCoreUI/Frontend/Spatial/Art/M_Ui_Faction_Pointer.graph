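# Translucent UI material for a faction pointer, drawing either an arrow or an adjustable four-sided shape procedurally, without textures.
# A static switch, IsArrow, picks between two oppositely rotated, sharpened UV edges and a box mask with per-side angle, offset and feather controls.
# HoverAnim blends two colours and their alphas.
# The blended alpha is inverted and subtracted from the mask, eroding it rather than scaling it.
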Material /FortniteCoreUI/Frontend/Spatial/Art/M_Ui_Faction_Pointer
graph {
 properties {
  bUsedWithStaticMesh=false
  blendMode=BLEND_Translucent
  materialDomain=MD_UI
  pixelDepthOffsetMode=PDOM_Legacy
  usedShadingModels=Default Lit
 }
 nodes [
  n1 = VectorParameter() {parameterName=RGB01, defaultValue=<rgba 0.8632 0.5776 0.0865 0.5>}
  n2 = VectorParameter() {parameterName=RGB02, defaultValue=<rgba 1 1 1 1>}
  n3 = ScalarParameter() {parameterName=HoverAnim}
  LinearInterpolate(A=@n1.RGB, B=@n2.RGB, Alpha=@n3)
  n4 = TextureCoordinate()
  n5 = ScalarParameter() {defaultValue=22.5, group=Degrees, parameterName=ArrowAngle}
  n6 = Rotator(Coordinate=@n4, Time=Divide(A=Multiply(A=@n5) {constB=-1}) {constB=360}) {centerX=1, speed=6.2832}
  n7 = ScalarParameter() {defaultValue=-0.5, parameterName=ArrowPosition}
  n8 = ScalarParameter() {defaultValue=128, group=Feather, parameterName=Sharpness}
  n9 = Multiply(A=ComponentMask(None=Add(A=@n6, B=@n7) {constB=-0.5}) {r=true}, B=@n8)
  n10 = Divide(A=@n5) {constB=360}
  n11 = ComponentMask(None=Add(A=Rotator(Coordinate=@n4, Time=@n10) {centerX=1, speed=6.2832}, B=@n7) {constB=-0.5}) {r=true}
  n12 = DotProduct(A=Saturate(None=OneMinus(None=@n9)), B=Saturate(None=Multiply(A=@n11, B=@n8)))
  n13 = ScalarParameter() {group=Degrees, parameterName=TopAngle}
  n14 = MaterialFunctionCall() {materialFunction=&GetUserInterfaceUV}
  n15 = Add(
   A=Rotator(Coordinate=@n14.Normalized UV, Time=Divide(A=@n13) {constB=360}) {centerX=1, speed=6.2832}
   B=ScalarParameter() {defaultValue=-0.1, group=Offset, parameterName=TopOffset}
  )
  n16 = ScalarParameter() {defaultValue=128, group=Feather, parameterName=HorizontalSharpness}
  n17 = Saturate(None=Multiply(A=ComponentMask(None=@n15) {g=true}, B=@n16))
  n18 = Divide(A=ScalarParameter() {group=Degrees, parameterName=BottomAngle}) {constB=360}
  n19 = ComponentMask(None=Add(
    A=Rotator(Coordinate=@n14.Normalized UV, Time=@n18) {centerX=0, speed=6.2832}
    B=ScalarParameter() {defaultValue=0.1, group=Offset, parameterName=BottomOffset}
   )) {g=true}
  n20 = Saturate(None=Multiply(A=OneMinus(None=@n19), B=@n16))
  n21 = ScalarParameter() {group=Degrees, parameterName=LeftAngle}
  n22 = Add(
   A=Rotator(Coordinate=@n14.9-Slice UV, Time=Divide(A=@n21) {constB=360}) {speed=6.2832}
   B=ScalarParameter() {defaultValue=-0.1, group=Offset, parameterName=LeftOffset}
  )
  n23 = ScalarParameter() {defaultValue=32, group=Feather, parameterName=VerticalSharpness}
  n24 = Saturate(None=Multiply(A=ComponentMask(None=@n22) {r=true}, B=@n23))
  n25 = Divide(A=ScalarParameter() {group=Degrees, parameterName=RightAngle}) {constB=360}
  n26 = ComponentMask(None=Add(
    A=Rotator(Coordinate=@n14.9-Slice UV, Time=@n25) {centerX=0, speed=6.2832}
    B=ScalarParameter() {defaultValue=0.1, group=Offset, parameterName=RightOffset}
   )) {r=true}
  n27 = Saturate(None=Multiply(A=OneMinus(None=@n26), B=@n23))
  n28 = Multiply(A=Saturate(None=Multiply(A=@n17, B=@n20)), B=Saturate(None=Multiply(A=@n24, B=@n27)))
  n29 = LinearInterpolate(A=@n1.A, B=@n2.A, Alpha=@n3)
  Saturate(None=Subtract(A=StaticSwitchParameter(True=@n12, False=@n28) {parameterName=IsArrow}, B=OneMinus(None=@n29)))
 ]
}
