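# Translucent UI material for a separator in the spatial frontend, drawing background, highlight and hover bands procedurally from UVs without textures.
# Each band mask offsets UVs, multiplies by a sharpness parameter and saturates for feathered edges.
# Color uses a locked/unlocked background blend inside an inset region and the highlight color elsewhere.
# Opacity sums the three bands, trimmed top and bottom, with HoverAnimControl growing the hover band vertically from the center.
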Material /FortniteCoreUI/Frontend/Spatial/Art/M_Ui_Spacial_Seaparator
graph {
 properties {
  bUsedWithStaticMesh=false
  blendMode=BLEND_Translucent
  materialDomain=MD_UI
  pixelDepthOffsetMode=PDOM_Legacy
  usedShadingModels=Default Lit
 }
 nodes [
  n1 = VectorParameter() {group=Highlight, parameterName=Highlight, defaultValue=<rgba 0 0.9761 1 0.5>}
  n2 = ScalarParameter() {defaultValue=-0.3, group=Background, parameterName=Background_LeftOffset}
  n3 = TextureCoordinate()
  n4 = Add(A=@n3, B=Add(A=@n2) {constB=0.05})
  n5 = ScalarParameter() {defaultValue=128, group=Feather, parameterName=VerticalSharpness}
  n6 = Saturate(None=Multiply(A=ComponentMask(None=@n4) {r=true}, B=@n5))
  n7 = ScalarParameter() {defaultValue=0.15, group=Background, parameterName=Background_RightOffset}
  n8 = ComponentMask(None=Add(A=@n3, B=Add(A=@n7) {constB=-0.05})) {r=true}
  n9 = Saturate(None=Multiply(A=OneMinus(None=@n8), B=@n5))
  LinearInterpolate(
   A=@n1.RGB
   B=LinearInterpolate(
    A=VectorParameter() {group=Background, parameterName=Background_Locked, defaultValue=<rgba 0.0331 0.0452 0.0931 1>}.RGB
    B=VectorParameter() {group=Background, parameterName=Background_unlocked, defaultValue=<rgba 0 0.0079 0.1927 1>}.RGB
    Alpha=ScalarParameter() {defaultValue=1, group=Background, parameterName=IsUnlocked}
   )
   Alpha=Saturate(None=Multiply(A=@n6, B=@n9))
  )
  n10 = TextureCoordinate()
  n11 = Add(A=@n10, B=ScalarParameter() {defaultValue=-0.035, group=Highlight, parameterName=Highlight_LeftOffset})
  n12 = ScalarParameter() {defaultValue=128, group=Feather, parameterName=VerticalSharpness}
  n13 = Saturate(None=Multiply(A=ComponentMask(None=@n11) {r=true}, B=@n12))
  n14 = ComponentMask(None=Add(A=@n10, B=ScalarParameter() {defaultValue=0.77, group=Highlight, parameterName=Highlight_RightOffset})) {r=true}
  n15 = Saturate(None=Multiply(A=OneMinus(None=@n14), B=@n12))
  n16 = Multiply(A=Saturate(None=Multiply(A=@n13, B=@n15)), B=@n1.A)
  n17 = TextureCoordinate()
  n18 = ComponentMask(None=Add(A=@n17, B=ScalarParameter() {defaultValue=-0.9, group=Hover, parameterName=Hover_LeftOffset})) {r=true}
  n19 = ScalarParameter() {defaultValue=128, group=Feather, parameterName=VerticalSharpness}
  n20 = ScalarParameter() {defaultValue=0.035, group=Hover, parameterName=Hover_RightOffset}
  n21 = OneMinus(None=ComponentMask(None=Add(A=@n17, B=@n20)) {r=true})
  n22 = Multiply(A=Saturate(None=Multiply(A=@n18, B=@n19)), B=Saturate(None=Multiply(A=@n21, B=@n19)))
  n23 = ScalarParameter() {defaultValue=0.5, group=Hover, parameterName=HoverAnimControl, sliderMax=1}
  n24 = ComponentMask(None=Add(A=@n17, B=LinearInterpolate(Alpha=@n23) {constA=-0.5, constB=0})) {g=true}
  n25 = ScalarParameter() {defaultValue=128, group=Feather, parameterName=HorizontalSharpness}
  n26 = LinearInterpolate(Alpha=@n23) {constA=0.5, constB=0}
  n27 = OneMinus(None=ComponentMask(None=Add(A=@n17, B=@n26)) {g=true})
  n28 = Multiply(A=Saturate(None=Multiply(A=@n24, B=@n25)), B=Saturate(None=Multiply(A=@n27, B=@n25)))
  n29 = Multiply(A=Multiply(A=Saturate(None=@n22), B=Saturate(None=@n28)), B=ScalarParameter() {defaultValue=1, group=Hover, parameterName=Hover_Opacity})
  n30 = ScalarParameter() {defaultValue=128, group=Feather, parameterName=VerticalSharpness}
  n31 = Multiply(A=ComponentMask(None=Add(A=@n3, B=@n2)) {r=true}, B=@n30)
  n32 = Add(A=@n3, B=@n7)
  n33 = Multiply(A=OneMinus(None=ComponentMask(None=@n32) {r=true}), B=@n30)
  n34 = Saturate(None=Multiply(A=Saturate(None=@n31), B=Saturate(None=@n33)))
  n35 = ScalarParameter() {defaultValue=0.5, group=Background, parameterName=Background_Opacity}
  n36 = TextureCoordinate()
  n37 = Add(A=@n36, B=ScalarParameter() {defaultValue=-0.035, group=Offset, parameterName=TopOffset})
  n38 = ScalarParameter() {defaultValue=128, group=Feather, parameterName=HorizontalSharpness}
  n39 = Saturate(None=Multiply(A=ComponentMask(None=@n37) {g=true}, B=@n38))
  n40 = ComponentMask(None=Add(A=@n36, B=ScalarParameter() {defaultValue=0.035, group=Offset, parameterName=BottomOffset})) {g=true}
  n41 = Saturate(None=Multiply(A=OneMinus(None=@n40), B=@n38))
  Multiply(A=Add(A=Add(A=@n16, B=@n29), B=Multiply(A=@n34, B=@n35)), B=Saturate(None=Multiply(A=@n39, B=@n41)))
 ]
}
