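# Translucent UI material that builds a one-sided horizontal gradient, named as a vignette for the spatial frontend UI.
# The inverted U texture coordinate is scaled by Gradient size, shaped by a Gradient Decay exponent, then clamped, so it is strongest at the left edge.
# The alpha of the RGBA parameter multiplies that falloff to set its overall strength.
# The mask is fully procedural, with no texture samples.
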Material /FortniteCoreUI/Frontend/Spatial/Art/M_Ui_Spatial_Gradient_Vignette
graph {
 properties {
  bUsedWithStaticMesh=false
  blendMode=BLEND_Translucent
  materialDomain=MD_UI
  pixelDepthOffsetMode=PDOM_Legacy
  usedShadingModels=Default Lit
 }
 nodes [
  n1 = TextureCoordinate()
  n2 = Multiply(A=OneMinus(None=ComponentMask(None=@n1) {r=true}), B=ScalarParameter() {defaultValue=1.25, parameterName=Gradient size})
  n3 = ScalarParameter() {defaultValue=1, parameterName=Gradient Decay}
  Multiply(A=VectorParameter() {parameterName=RGBA, defaultValue=<rgba 0.5 0.5 0.5 0.5>}.A, B=Saturate(None=Power(Base=@n2, Exp=@n3)))
 ]
}
