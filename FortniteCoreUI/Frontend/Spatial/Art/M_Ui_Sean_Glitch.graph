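# UI material function that renders the ItemIcon texture with an animated chromatic-aberration glitch, outputting RGB and Alpha.
# Multiplied sine waves drive opposing horizontal UV offsets for red and blue samples around a static green sample.
# Three panning sine scanline layers, built on square-to-screen-space coordinates, add further horizontal distortion to red and blue.
# Alpha sums all three samples, UseGlitch blends against the clean icon, and WhiteFill fades RGB to white.
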MaterialFunction /FortniteCoreUI/Frontend/Spatial/Art/M_Ui_Sean_Glitch
graph {nodes [
  n1 = MaterialFunctionCall(UVs=TextureCoordinate(), Texture Scale=ScalarParameter() {defaultValue=1, group=Icon, parameterName=Icon Scale}) {materialFunction=&ScaleUVsByCenter}
  n2 = Add(
   A=Multiply(A=Time(), B=ScalarParameter() {defaultValue=3, group=Global, parameterName=GlobalTimeControl})
   B=ScalarParameter() {group=Global, parameterName=RandomizeTimeOffset}
  )
  n3 = ScalarParameter() {defaultValue=0.5, group=Portrait, parameterName=RedChannel_Time_Sine_1}
  n4 = Add(A=Sine(None=Multiply(A=@n2, B=@n3)))
  n5 = ScalarParameter() {defaultValue=4, group=Portrait, parameterName=RedChannel_Time_Sine_2}
  n6 = Add(A=Sine(None=Multiply(A=@n2, B=@n5)))
  n7 = Sine(None=Multiply(A=@n2, B=ScalarParameter() {defaultValue=1, group=Portrait, parameterName=RedChannel_Time_Sine_3}))
  n8 = DotProduct(A=DotProduct(A=LinearInterpolate(Alpha=@n4), B=LinearInterpolate(Alpha=@n6)), B=LinearInterpolate(Alpha=Add(A=@n7)))
  n9 = AppendVector(
   A=LinearInterpolate(
    A=ScalarParameter() {defaultValue=0.0012, group=Portrait, parameterName=RedChannel_Min_Offset}
    B=ScalarParameter() {defaultValue=0.005, group=Portrait, parameterName=RedChannel_Max_Offset}
    Alpha=@n8
   )
   B=Constant()
  )
  n10 = MaterialFunctionCall() {materialFunction=/FortUILibrary/Foundation/Art_ToolKit/Tools/MF_UI_SquareToScreenSpace}
  n11 = Add(A=Time(), B=ScalarParameter() {group=Portrait, parameterName=RandomizeTimeOffset})
  n12 = Panner(Coordinate=@n10.Y_Priority, Time=@n11) {speedY=-0.0125}
  n13 = ScalarParameter() {defaultValue=64, group=Portrait, parameterName=Lines_1_Amount}
  n14 = ComponentMask(None=Sine(None=Multiply(A=@n12, B=@n13))) {g=true}
  n15 = Panner(Coordinate=@n10.Y_Priority, Time=@n11) {speedY=0.0063}
  n16 = ScalarParameter() {defaultValue=128, group=Portrait, parameterName=Lines_2_Amount}
  n17 = ComponentMask(None=Sine(None=Multiply(A=@n15, B=@n16))) {g=true}
  n18 = Sine(None=Multiply(
    A=Panner(Coordinate=@n10.Y_Priority, Time=@n11) {speedY=0.1}
    B=ScalarParameter() {defaultValue=16, group=Portrait, parameterName=Lines_3_Amount}
   ))
  n19 = Saturate(None=DotProduct(A=DotProduct(A=@n14, B=@n17), B=ComponentMask(None=@n18) {g=true}))
  n20 = ScalarParameter() {defaultValue=8, group=Portrait, parameterName=ScanLines_Softness}
  n21 = Multiply(A=Saturate(None=Multiply(A=@n19, B=@n20)), B=ScalarParameter() {defaultValue=1, group=Portrait, parameterName=ScanLines_Amount})
  n22 = LinearInterpolate(
   B=ScalarParameter() {defaultValue=0.0025, group=Portrait, parameterName=Red_Channel_Distortion_Lines_Amount}
   Alpha=DotProduct(A=@n21, B=@n8)
  ) {constB=0.1}
  n23 = Constant()
  n24 = TextureSampleParameter2D(UVs=Add(A=Add(A=@n1.UVs, B=@n9), B=AppendVector(A=@n22, B=@n23))) {
   group=Icon
   parameterName=ItemIcon
   samplerSource=SSM_Clamp_WorldGroupSettings
   texture=/Game/UI/Foundation/Textures/Icons/Heroes/Athena/Soldier/T-Soldier-HID-605-Athena-Commando-M-TourBus
  }
  n25 = ScalarParameter() {defaultValue=0.0012, group=Portrait, parameterName=BlueChannel_Min_Offset}
  n26 = ScalarParameter() {defaultValue=0.005, group=Portrait, parameterName=BlueChannel_Max_Offset}
  n27 = ScalarParameter() {defaultValue=0.25, group=Portrait, parameterName=BlueChannel_Time_Sine_1}
  n28 = Add(A=Sine(None=Multiply(A=@n2, B=@n27)))
  n29 = ScalarParameter() {defaultValue=6, group=Portrait, parameterName=BlueChannel_Time_Sine_2}
  n30 = Add(A=Sine(None=Multiply(A=@n2, B=@n29)))
  n31 = Sine(None=Multiply(A=@n2, B=ScalarParameter() {defaultValue=0.75, group=Portrait, parameterName=BlueChannel_Time_Sine_3}))
  n32 = DotProduct(A=DotProduct(A=LinearInterpolate(Alpha=@n28), B=LinearInterpolate(Alpha=@n30)), B=LinearInterpolate(Alpha=Add(A=@n31)))
  n33 = AppendVector(A=Multiply(A=LinearInterpolate(A=@n25, B=@n26, Alpha=@n32)) {constB=-1}, B=Constant())
  n34 = LinearInterpolate(
   B=Multiply(A=ScalarParameter() {defaultValue=0.0025, group=Portrait, parameterName=Blue_Channel_Distortion_Lines_Amount}) {constB=-1}
   Alpha=DotProduct(A=@n21, B=@n32)
  ) {constB=-0.003}
  n35 = Constant()
  n36 = TextureSampleParameter2D(UVs=Add(A=Add(A=@n1.UVs, B=@n33), B=AppendVector(A=@n34, B=@n35))) {
   group=Icon
   parameterName=ItemIcon
   samplerSource=SSM_Clamp_WorldGroupSettings
   texture=&T-Soldier-HID-605-Athena-Commando-M-TourBus
  }
  n37 = TextureSampleParameter2D(UVs=MaterialFunctionCall(Texture Scale=ScalarParameter() {defaultValue=1, group=Icon, parameterName=Icon Scale}) {materialFunction=&ScaleUVsByCenter}.UVs) {
   group=Icon
   parameterName=ItemIcon
   samplerSource=SSM_Clamp_WorldGroupSettings
   texture=&T-Soldier-HID-605-Athena-Commando-M-TourBus
  }
  n38 = Add(A=@n37.A, B=Add(A=@n24.A, B=@n36.A))
  n39 = ScalarParameter() {defaultValue=1, group=Global, parameterName=UseGlitch}
  FunctionOutput(None=Saturate(None=LinearInterpolate(A=@n37.A, B=@n38, Alpha=@n39))) {outputName=Alpha}
  n40 = Saturate(None=MaterialFunctionCall(R=@n24.R, G=@n37.G, B=@n36.B) {materialFunction=&AppendMany}.RGB)
  FunctionOutput(None=LinearInterpolate(A=LinearInterpolate(A=@n37.RGB, B=@n40, Alpha=@n39), B=Constant() {r=1}, Alpha=ScalarParameter() {parameterName=WhiteFill})) {outputName=RGB}
 ]}
