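# Translucent UI material for a spatial frontend item card, drawing a glitch-function icon over a radial gradient that blends between locked and unlocked colors.
# The icon's alpha is reduced by animatable opacity and a vertical falloff, then composites it onto the gradient.
# An animated, rotated-UV mask paints a solid-color divet.
# Sharpened UV offsets form an inset rectangular alpha, and UseBackground switches between the full card and icon alone.
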Material /FortniteCoreUI/Frontend/Spatial/Art/M_Ui_Spatial_itemCard
graph {
 properties {
  bUsedWithStaticMesh=false
  blendMode=BLEND_Translucent
  materialDomain=MD_UI
  pixelDepthOffsetMode=PDOM_Legacy
  usedShadingModels=Default Lit
 }
 nodes [
  n1 = ScalarParameter() {defaultValue=1, group=Gradient, parameterName=IsUnlocked}
  n2 = LinearInterpolate(
   A=LinearInterpolate(
    A=VectorParameter() {group=Gradient, parameterName=LockedGradientStart, defaultValue=<rgba 0.0343 0.0467 0.0953 1>}.RGB
    B=VectorParameter() {group=Gradient, parameterName=UnlockedGradientStart, defaultValue=<rgba 0.0075 0.0176 0.2086 1>}.RGB
    Alpha=@n1
   )
   B=LinearInterpolate(
    A=VectorParameter() {group=Gradient, parameterName=LockedGradientStop, defaultValue=<rgba 0.0782 0.1301 0.3712 1>}.RGB
    B=VectorParameter() {group=Gradient, parameterName=UnlockedGradientStop, defaultValue=<rgba 0 0.0908 1 1>}.RGB
    Alpha=@n1
   )
   Alpha=MaterialFunctionCall(
    Radius=ScalarParameter() {defaultValue=0.6, group=Gradient, parameterName=GradientRadius}
    Density=ScalarParameter() {defaultValue=1.2, group=Gradient, parameterName=GradientDensity}
   ) {materialFunction=&RadialGradientExponential}.RadialGradientExponential
  )
  n3 = MaterialFunctionCall() {materialFunction=&/M_Ui_Sean_Glitch}
  n4 = OneMinus(None=ScalarParameter() {defaultValue=0.75, group=Icon, parameterName=IconOpacity})
  n5 = Clamp(None=LinearInterpolate(A=Clamp(None=@n4), Alpha=ScalarParameter() {group=Icon, parameterName=AnimateIconOpacity}) {constB=0})
  n6 = ScalarParameter() {defaultValue=-0.5, group=Icon, parameterName=Gradient_Fade_Falloff_Location}
  n7 = ComponentMask(None=TextureCoordinate()) {g=true}
  n8 = Power(
   Base=Saturate(None=LinearInterpolate(A=@n6, Alpha=@n7))
   Exp=ScalarParameter() {defaultValue=2, group=Icon, parameterName=Gradient_Fade_Falloff_Rate}
  )
  n9 = Multiply(A=Saturate(None=Subtract(A=@n3.Alpha, B=@n5)), B=OneMinus(None=Saturate(None=@n8)))
  n10 = TextureCoordinate()
  n11 = Divide(A=LinearInterpolate(Alpha=ScalarParameter() {group=divet, parameterName=DivetAnim, sliderMax=1}) {constA=-45, constB=-90}) {constB=360}
  n12 = ComponentMask(None=Add(
    A=Rotator(Coordinate=@n10, Time=@n11) {centerX=0, speed=6.2832}
    B=ScalarParameter() {defaultValue=-0.65, group=divet, parameterName=DivetOffset}
   )) {g=true}
  n13 = ScalarParameter() {defaultValue=64, group=Offset, parameterName=Sharpness}
  LinearInterpolate(
   A=LinearInterpolate(
    A=@n3.RGB
    B=LinearInterpolate(A=@n2, B=@n3.RGB, Alpha=@n9)
    Alpha=ScalarParameter() {defaultValue=1, parameterName=UseBackground, sliderMax=1}
   )
   B=VectorParameter() {group=divet, parameterName=DivetColor, defaultValue=<rgba 0 0 0 1>}.RGB
   Alpha=Saturate(None=Multiply(A=@n12, B=@n13))
  )
  n14 = Add(A=@n10, B=ScalarParameter() {defaultValue=-0.035, group=Offset, parameterName=TopOffset})
  n15 = Saturate(None=Multiply(A=ComponentMask(None=@n14) {g=true}, B=@n13))
  n16 = ComponentMask(None=Add(A=@n10, B=ScalarParameter() {defaultValue=0.035, group=Offset, parameterName=BottomOffset})) {g=true}
  n17 = Saturate(None=Multiply(A=OneMinus(None=@n16), B=@n13))
  n18 = Add(A=@n10, B=ScalarParameter() {defaultValue=-0.035, group=Offset, parameterName=LeftOffset})
  n19 = Saturate(None=Multiply(A=ComponentMask(None=@n18) {r=true}, B=@n13))
  n20 = ComponentMask(None=Add(A=@n10, B=ScalarParameter() {defaultValue=0.035, group=Offset, parameterName=RightOffset})) {r=true}
  n21 = Saturate(None=Multiply(A=OneMinus(None=@n20), B=@n13))
  n22 = Multiply(A=Saturate(None=Multiply(A=@n15, B=@n17)), B=Saturate(None=Multiply(A=@n19, B=@n21)))
  Multiply(
   A=Clamp(None=ScalarParameter() {defaultValue=1, parameterName=OverallOpacity})
   B=LinearInterpolate(A=@n9, B=Saturate(None=@n22), Alpha=ScalarParameter() {defaultValue=1, parameterName=UseBackground, sliderMax=1})
  )
 ]
}
