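# Translucent UI material for the faction switcher graphic, coloring a channel-packed texture with animated stripes and a light sweep.
# Inner and outer colors form panning, angled sine bars with slight time-based flicker, washed toward the outer color by a screen-centered horizontal gradient.
# Texture red minus green masks in the outer color with an optional looping, aspect-corrected screen-space sweep.
# Opacity combines the red channel with boosted blue plus a Shadow offset.
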Material /FortniteCoreUI/Frontend/Spatial/Art/T_Ui_Faction_switcher_02_Mat
graph {
 properties {
  bUsedWithStaticMesh=false
  blendMode=BLEND_Translucent
  materialDomain=MD_UI
  pixelDepthOffsetMode=PDOM_Legacy
  usedShadingModels=Default Lit
 }
 nodes [
  n1 = VectorParameter() {parameterName=InerColor, defaultValue=<rgba 0.2918 0.1413 0.007 1>}
  n2 = VectorParameter() {parameterName=Outer color, defaultValue=<rgba 0.8632 0.5776 0.0865 1>}
  n3 = MaterialFunctionCall!(
   Rotation_Fraction=ScalarParameter() {defaultValue=-0.05, group=Bars, parameterName=AngledBars_Rotation}
   TexCoord=MaterialFunctionCall() {materialFunction=/FortUILibrary/Foundation/Art_ToolKit/Tools/MF_UI_SquareToScreenSpace}.X_Priority
  ) {materialFunction=/FortUILibrary/Foundation/Art_ToolKit/AngledBoxes/MF_UI_Rotator_Fractions}
  n4 = AppendVector(A=ScalarParameter() {defaultValue=-0.025, group=Bars, parameterName=X_PanSpeed}, B=ScalarParameter() {group=Bars, parameterName=Y_PanSpeed})
  n5 = Sine(None=Multiply(A=Panner(Coordinate=@n3.Result, Speed=@n4), B=ScalarParameter() {defaultValue=48, group=Bars, parameterName=AngledBars_Amount}))
  n6 = Saturate(None=Multiply(A=ComponentMask(None=@n5) {r=true}, B=ScalarParameter() {defaultValue=4, group=Bars, parameterName=Bar_Softness}))
  n7 = Multiply(A=Time(), B=ScalarParameter() {defaultValue=0.6667, group=Bars, parameterName=Overall_TimeMultiplier})
  n8 = Constant()
  n9 = Constant() {r=1}
  n10 = If(A=LinearInterpolate(Alpha=Sine(None=@n7)) {constA=-3}, B=@n8, "A > B"=@n9, "A == B"=@n9, "A < B"=@n8)
  n11 = Multiply(A=@n7, B=ScalarParameter() {defaultValue=3, group=Bars, parameterName=Sine2_Time})
  n12 = LinearInterpolate(Alpha=Add(A=Sine(None=@n11)))
  n13 = LinearInterpolate(
   A=ScalarParameter() {defaultValue=0.05, group=Bars, parameterName=Bars_Alpha}
   B=Saturate(None=DotProduct(A=@n10, B=@n12))
   Alpha=ScalarParameter() {defaultValue=0.05, group=Bars, parameterName=Use_OscillatingTimeMultiplier}
  )
  n14 = Saturate(None=Add(A=Multiply(A=@n6, B=@n13), B=ScalarParameter() {group=Bars, parameterName=Gaps_Alpha}))
  n15 = ComponentMask(None=MaterialFunctionCall(Texture Dimensions=TextureCoordinate()) {materialFunction=&ScreenAlignedUVs}."X 100%, Y 100%") {r=true}
  n16 = DotProduct(A=OneMinus(None=@n15), B=@n15)
  n17 = ScalarParameter() {defaultValue=4, group=ScreenspaceGradient, parameterName=FallOffPoint}
  n18 = Power(Base=Saturate(None=Multiply(A=@n16, B=@n17)), Exp=ScalarParameter() {defaultValue=2, group=ScreenspaceGradient, parameterName=FallOffRate})
  n19 = VectorParameter() {group=lightSweep, parameterName=lightsweepColor, defaultValue=<rgba 1 1 1 1>}
  n20 = MaterialFunctionCall(Float2=MaterialFunctionCall() {materialFunction=&ScreenResolution}.Visible Resolution) {materialFunction=&BreakOutFloat2Components}
  n21 = Divide(A=@n20.R, B=Max(A=@n20.G) {constB=0.1})
  n22 = Constant() {r=1}
  n23 = Rotator(
   Coordinate=Multiply(
    A=AppendVector(A=@n21, B=@n22)
    B=MaterialFunctionCall() {materialFunction=&ScreenAlignedUVs}."X 100%, Y 100%"
   )
   Time=ScalarParameter() {defaultValue=-0.125, group=lightSweep, parameterName=Angle}
  ) {speed=6.2832}
  n24 = Multiply(A=Time(), B=ScalarParameter() {defaultValue=0.25, group=lightSweep, parameterName=Time Factor})
  n25 = ScalarParameter() {defaultValue=1, group=lightSweep, parameterName=Period}
  n26 = Multiply(A=LinearInterpolate(
    A=ScalarParameter() {defaultValue=-0.5, parameterName=left offset}
    B=ScalarParameter() {defaultValue=1.5, parameterName=right offset}
    Alpha=Fmod(A=@n24, B=@n25)
   ) {constA=-0.5, constB=1.5}) {constB=-1}
  n27 = ComponentMask(None=Sine(None=Add(A=@n23, B=@n26)) {period=2}) {r=true}
  n28 = ScalarParameter() {defaultValue=0.99, group=lightSweep, parameterName=EdgeOffset}
  n29 = Saturate(None=Multiply(A=Subtract(A=@n27, B=@n28) {constB=0.5}, B=ScalarParameter() {defaultValue=1024, group=lightSweep, parameterName=EdgeFallOff}) {constB=64})
  n30 = TextureSample() {texture=&/T_Ui_Faction_switcher_02}
  LinearInterpolate(
   A=LinearInterpolate(A=LinearInterpolate(A=@n1.RGB, B=@n2.RGB, Alpha=@n14), B=@n2.RGB, Alpha=Saturate(None=@n18))
   B=LinearInterpolate(
    A=@n2.RGB
    B=LinearInterpolate(A=@n2.RGB, B=@n19.RGB, Alpha=@n29)
    Alpha=ScalarParameter() {defaultValue=1, group=lightSweep, parameterName=UseLightSweep}
   )
   Alpha=DotProduct(A=@n30.R, B=OneMinus(None=@n30.G))
  )
  n31 = Add(A=@n30.B, B=ScalarParameter() {parameterName=Shadow})
  Saturate(None=Add(A=Multiply(A=@n31) {constB=1.5}, B=@n30.R))
 ]
}
